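# Niagara idle effect for the Nautilus player character cosmetic, providing an orange glowing headlight.
# A single Headlight emitter burst-spawns particles that persist past their lifetime and feed sprite, light and mesh renderers.
# A waveform pulses the color alpha, multiplied by a float from User.bIsHeadHidden that zeroes it when the head is hidden.
# The update also writes dynamic material parameters, light attributes and a glow camera offset.
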NiagaraSystem /BRCosmetics/Characters/Player/Female/Medium/Bodies/F_MED_Nautilus/FX/NS_Character_Nautilus_Idle
dependencies {usedDataInterfaces [], usedDynamicInputs [], usedModules [&SystemState]}
emitters {Headlight {
  emitterUpdateScript [
   &EmitterState(
    Inactive Response=::NewEnumerator1
    Life Cycle Mode=::NewEnumerator1
    Loop Behavior=::NewEnumerator1
    Loop Duration Mode=::NewEnumerator1
    UseLoopDelay=true
   )
   &SpawnBurst_Instantaneous(Use Spawn Probability=false, UseLoopCountLimit=true)
  ]
  particleSpawnScript [&InitializeParticle(
    Color=<rgba 1 0.5 0 1>
    Lifetime=1.25
    Mesh Scale=<1.15 1.15 1.15>
    Mesh Scale Mode=::NewEnumerator4
    Sprite Size=<11 11>
    Sprite Size Mode=::NewEnumerator4
    UsePositionOffset=false
    Write Material Random=true
   )]
  particleUpdateScript [
   &ParticleState(Kill Particles When Lifetime Has Elapsed=false)
   &ScaleColor(
    Scale Alpha=&Multiply_Float(
     A=&Waveform("Waveform [1]"=::NewEnumerator0, "[1] Frequency"=0.5, "[1] Amplitude Min/Max"=<0.5 1>)
     B=&MakeCustomFloatFromBool(Bool=$User.bIsHeadHidden, True Float=0, False Float=1)
    )
    Scale RGB=<1.5 1.5 1.5>
    ScaleA=true
    ScaleRGB=true
   )
   &DynamicMaterialParameters(
    Index 0 Param 1=0.8
    Index 0 Param 2=2
    Index 0 Param 3=1
    Index 0 Param 4=1
    Module.Param0WriteEnabled=true
    Module.Param1WriteEnabled=true
    Module.Param2WriteEnabled=true
    Module.Param3WriteEnabled=true
   )
   &FN_Light_Attributes(
    Color=$Particles.Color
    Create Color Attribute=true
    Create Exponent Attribute=true
    Create Radius Attribute=true
    Exponent=2
    Radius=15
    Scale Color=2
   )
   /SetVariables(Particles.Glow Camera Offset=2.5) {isSetParameters=true}
  ]
  rendererClasses [
   &NiagaraSpriteRendererProperties
   &NiagaraLightRendererProperties
   &NiagaraMeshRendererProperties
  ]
  renderers [
   {rendererClass=&NiagaraSpriteRendererProperties}
   {rendererClass=&NiagaraLightRendererProperties, rendererIndex=1}
   {rendererClass=&NiagaraMeshRendererProperties, rendererIndex=2}
  ]
 }}
summary {
 systemName=NS_Character_Nautilus_Idle
 emitters [{
   emitterName=Headlight
   rendererClasses [
    &NiagaraSpriteRendererProperties
    &NiagaraLightRendererProperties
    &NiagaraMeshRendererProperties
   ]
  }]
 userVariables [User.bIsHeadHidden {type=NiagaraBool}]
}
