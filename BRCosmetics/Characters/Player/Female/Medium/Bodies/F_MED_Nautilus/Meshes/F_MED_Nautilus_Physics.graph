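# Physics asset for the F_MED_Nautilus female player body, covering the weight belt, its rings and a leg seaweed strand.
# The belt master is angularly locked to the pelvis, with a tightly limited front ring and a free-rotating back link carrying a locked ring.
# A two-link seaweed chain hangs from the pelvis, swinging mainly on one axis.
# Pelvis and left thigh are the only non-dyn bodies, with the thigh joint free.
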PhysicsAsset /BRCosmetics/Characters/Player/Female/Medium/Bodies/F_MED_Nautilus/Meshes/F_MED_Nautilus_Physics
bodies {
 dyn_belt_ring_front_bottom [{shapeType=Box, extentX=0.8225, extentY=4.3839, extentZ=4.3894, center=<1.6996 0.0945 0.0001>, rotation=<rot 0 -90 0>}]
 dyn_seaweed_leg01 [{shapeType=Box, extentX=6.9985, extentY=3.3422, extentZ=1.7142, center=<3.0494 -0.3092 -0.0168>}]
 dyn_seaweed_leg02 [{shapeType=Box, extentX=5.9095, extentY=2.8992, extentZ=1.7, center=<3.7644 0.6585 -0.2394>}]
 dyn_weight_belt_back_ring [{shapeType=Box, extentX=0.6751, extentY=5.1865, extentZ=3.6244, center=<-0.0208 -0.163 0.0022>, rotation=<rot 0 -100.4163 -89.3204>}]
 dyn_weight_belt_back_ring_const [{shapeType=Box, extentX=1.01, extentY=4.4411, extentZ=6.9721, center=<0 0.9944 0>, rotation=<rot 0 -97.1704 0>}]
 dyn_weight_belt_master [{shapeType=Box, extentX=12.5062, extentY=26.0779, extentZ=24.9894, center=<6.0959 -0.24 0.822>}]
 pelvis [
  {shapeType=Capsule, length=16, radius=14, center=<-3.3369 1.7177 0>}
  {shapeType=Box, extentX=16.8133, extentY=16.9923, extentZ=33.4495, center=<3.9321 0.0571 -0.5345>}
 ]
 thigh_l [{shapeType=Capsule, length=33.4602, radius=7.5964, center=<-18.333 1.6134 -0.1916>, rotation=<rot 90 -13.4732 -15.6335>}]
}
constraints [
 {
  bone1Name=dyn_weight_belt_master
  bone2Name=pelvis
  swing1LimitDegrees=45
  swing1Motion=Locked
  swing2LimitDegrees=45
  swing2Motion=Locked
  twistLimitDegrees=45
  twistMotion=Locked
 }
 {bone1Name=dyn_seaweed_leg01, bone2Name=pelvis, swing1LimitDegrees=1, swing2LimitDegrees=20, twistLimitDegrees=1}
 {bone1Name=dyn_seaweed_leg02, bone2Name=dyn_seaweed_leg01, swing1LimitDegrees=1, swing2LimitDegrees=20, twistLimitDegrees=1}
 {
  bone1Name=thigh_l
  bone2Name=pelvis
  swing1LimitDegrees=45
  swing1Motion=Free
  swing2LimitDegrees=45
  swing2Motion=Free
  twistLimitDegrees=45
  twistMotion=Free
 }
 {bone1Name=dyn_belt_ring_front_bottom, bone2Name=dyn_weight_belt_master, swing1LimitDegrees=10, swing2LimitDegrees=1, twistLimitDegrees=1}
 {
  bone1Name=dyn_weight_belt_back_ring_const
  bone2Name=dyn_weight_belt_master
  swing1LimitDegrees=45
  swing1Motion=Free
  swing2LimitDegrees=45
  swing2Motion=Free
  twistLimitDegrees=45
  twistMotion=Free
 }
 {
  bone1Name=dyn_weight_belt_back_ring
  bone2Name=dyn_weight_belt_back_ring_const
  swing1LimitDegrees=45
  swing1Motion=Locked
  swing2LimitDegrees=45
  swing2Motion=Locked
  twistLimitDegrees=45
  twistMotion=Locked
 }
]
