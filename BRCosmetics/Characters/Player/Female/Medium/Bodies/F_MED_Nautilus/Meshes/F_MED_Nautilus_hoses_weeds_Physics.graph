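# Physics asset for the female medium Nautilus body cosmetic, simulating its dangling hoses and seaweed.
# Capsules and boxes on pelvis, spine, right clavicle, right calf and left thigh form the body, linked by free-swinging constraints.
# Tight angular limits attach the hose and tube bones to the spine.
# Seaweed chains hang from the pelvis at front and back and from the right calf, each link narrowly limited.
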PhysicsAsset /BRCosmetics/Characters/Player/Female/Medium/Bodies/F_MED_Nautilus/Meshes/F_MED_Nautilus_hoses_weeds_Physics
bodies {
 calf_r [{shapeType=Capsule, length=30.601, radius=7.2413, center=<26.5188 5.317 0.3481>, rotation=<rot 2.7679 95.2066 -89.4759>}]
 clavicle_r [{shapeType=Capsule, length=10.1487, radius=5.6631, center=<-4.5359 1.067 1.8852>, rotation=<rot -87.0135 -0.0003 0.0003>}]
 dyn_clav_r_tube [{shapeType=Box, extentX=4.01, extentY=13.51, extentZ=2.51, center=<3.3995 0 -0.0865>}]
 dyn_front_seaweed_strand01 [{shapeType=Box, extentX=7.0034, extentY=1.4671, extentZ=4.0987, center=<3.0786 -0.2269 -0.2432>}]
 dyn_front_seaweed_strand02 [{shapeType=Box, extentX=11.0747, extentY=1.1225, extentZ=4.1255, center=<2.4241 0.232 -0.1889>}]
 dyn_front_seaweed_strand03 [{shapeType=Box, extentX=6.9981, extentY=0.4502, extentZ=2.5134, center=<4.3513 0.448 -0.0542>}]
 dyn_hose_front [{shapeType=Box, extentX=3.9126, extentY=4.4208, extentZ=4.7634, center=<6.5243 -11.075 0.707>, rotation=<rot 0 -90.3345 -87.3407>}]
 dyn_hose_under_arm [{shapeType=Box, extentX=7.2919, extentY=21.2676, extentZ=2.8713, center=<20.2106 -2.2171 -0.5097>, rotation=<rot 1.7037 -0.3624 -12.0158>}]
 dyn_seaweed_back_left01 [{shapeType=Box, extentX=7.9919, extentY=1.0761, extentZ=4.4434, center=<4.3753 0.3223 0.0964>}]
 dyn_seaweed_back_left02 [{shapeType=Box, extentX=6.1481, extentY=0.9545, extentZ=4.295, center=<2.8078 0.3074 0.2026>}]
 dyn_seaweed_back_left03 [{shapeType=Box, extentX=7.7615, extentY=1.5863, extentZ=2.4712, center=<3.0231 0.1405 -0.0376>}]
 dyn_seaweed_back_middle01 [{shapeType=Box, extentX=5.4204, extentY=6.3584, extentZ=12.9415, center=<5.5924 0.9333 -0.3642>, rotation=<rot 0 85.2002 -89.8517>}]
 dyn_seaweed_back_middle02 [{shapeType=Box, extentX=1.7165, extentY=4.0394, extentZ=5.9196, center=<3.7197 -0.2553 -0.4103>, rotation=<rot 0 -89.0901 -84.9814>}]
 dyn_seaweed_back_middle03 [{shapeType=Box, extentX=1.5, extentY=1.6913, extentZ=8.0249, center=<3.0822 -0.6856 -0.6489>, rotation=<rot 0 -94.0887 -89.4065>}]
 dyn_seaweed_back_right01 [{shapeType=Box, extentX=5.6343, extentY=7.2947, extentZ=10.8098, center=<5.2871 -0.2826 0.0728>, rotation=<rot 0.0499 90.2916 -89.4759>}]
 dyn_seaweed_back_right02 [{shapeType=Box, extentX=3.382, extentY=3.8531, extentZ=8.5155, center=<4.7911 0.3108 0.0636>, rotation=<rot 0 -93.6633 -88.8523>}]
 dyn_seaweed_back_right03 [{shapeType=Box, extentX=2.3416, extentY=3.6819, extentZ=10.6177, center=<5.0183 -1.3101 -0.3189>, rotation=<rot 0 -93.8174 -86.2628>}]
 dyn_seaweed_leg01 [{shapeType=Box, extentX=6.9985, extentY=3.3422, extentZ=1.7142, center=<3.0494 -0.3092 -0.0168>}]
 dyn_seaweed_leg02 [{shapeType=Box, extentX=5.9095, extentY=2.8992, extentZ=1.7, center=<3.7644 0.6585 -0.2394>}]
 pelvis [
  {shapeType=Capsule, length=16, radius=14, center=<-3.3369 1.7177 0>}
  {shapeType=Box, extentX=16.8133, extentY=19.0712, extentZ=33.4495, center=<3.9321 0.0571 -0.5345>}
 ]
 spine_03 [{shapeType=Box, extentX=5.01, extentY=2.01, extentZ=2.01, center=<4.0777 6.0734 -1.4088>, rotation=<rot 0 -101.6441 -74.3839>}]
 spine_05 [{shapeType=Box, extentX=22.3272, extentY=19.4662, extentZ=9.5926, center=<6.2274 1.897 -0.28>}]
 thigh_l [{shapeType=Capsule, length=33.4602, radius=7.5964, center=<-18.333 1.6134 -0.1916>, rotation=<rot 90 -13.4732 -15.6335>}]
}
constraints [
 {
  bone1Name=spine_05
  bone2Name=spine_03
  swing1LimitDegrees=45
  swing1Motion=Free
  swing2LimitDegrees=45
  swing2Motion=Free
  twistLimitDegrees=45
  twistMotion=Free
 }
 {bone1Name=dyn_hose_under_arm, bone2Name=spine_05, swing1LimitDegrees=5, swing2LimitDegrees=1, twistLimitDegrees=1}
 {bone1Name=dyn_clav_r_tube, bone2Name=spine_05, swing1LimitDegrees=3, swing2LimitDegrees=15, twistLimitDegrees=1}
 {
  bone1Name=clavicle_r
  bone2Name=spine_05
  swing1LimitDegrees=45
  swing1Motion=Free
  swing2LimitDegrees=45
  swing2Motion=Free
  twistLimitDegrees=45
  twistMotion=Free
 }
 {
  bone1Name=calf_r
  bone2Name=pelvis
  swing1LimitDegrees=45
  swing1Motion=Free
  swing2LimitDegrees=45
  swing2Motion=Free
  twistLimitDegrees=45
  twistMotion=Free
 }
 {bone1Name=dyn_seaweed_leg01, bone2Name=calf_r, swing1LimitDegrees=1, swing2LimitDegrees=20, twistLimitDegrees=1}
 {bone1Name=dyn_seaweed_leg02, bone2Name=dyn_seaweed_leg01, swing1LimitDegrees=1, swing2LimitDegrees=20, twistLimitDegrees=1}
 {bone1Name=dyn_front_seaweed_strand02, bone2Name=dyn_front_seaweed_strand01, swing1LimitDegrees=20, swing2LimitDegrees=5, twistLimitDegrees=0}
 {bone1Name=dyn_front_seaweed_strand03, bone2Name=dyn_front_seaweed_strand02, swing1LimitDegrees=20, swing2LimitDegrees=2, twistLimitDegrees=7}
 {bone1Name=pelvis, bone2Name=dyn_front_seaweed_strand01, swing1LimitDegrees=10, swing2LimitDegrees=1, twistLimitDegrees=1}
 {
  bone1Name=thigh_l
  bone2Name=pelvis
  swing1LimitDegrees=45
  swing1Motion=Free
  swing2LimitDegrees=45
  swing2Motion=Free
  twistLimitDegrees=45
  twistMotion=Free
 }
 {bone1Name=pelvis, bone2Name=dyn_seaweed_back_left01, swing1LimitDegrees=5, swing2LimitDegrees=15, twistLimitDegrees=1}
 {bone1Name=dyn_seaweed_back_left02, bone2Name=dyn_seaweed_back_left01, swing1LimitDegrees=10, swing2LimitDegrees=5, twistLimitDegrees=1}
 {bone1Name=dyn_seaweed_back_left03, bone2Name=dyn_seaweed_back_left02, swing1LimitDegrees=10, swing2LimitDegrees=5, twistLimitDegrees=1}
 {bone1Name=pelvis, bone2Name=dyn_seaweed_back_middle01, swing1LimitDegrees=10, swing2LimitDegrees=5, twistLimitDegrees=1}
 {bone1Name=dyn_seaweed_back_middle02, bone2Name=dyn_seaweed_back_middle01, swing1LimitDegrees=20, swing2LimitDegrees=8, twistLimitDegrees=1}
 {bone1Name=dyn_seaweed_back_middle03, bone2Name=dyn_seaweed_back_middle02, swing1LimitDegrees=20, swing2LimitDegrees=8, twistLimitDegrees=1}
 {bone1Name=pelvis, bone2Name=dyn_seaweed_back_right01, swing1LimitDegrees=15, swing2LimitDegrees=12, twistLimitDegrees=3}
 {bone1Name=dyn_seaweed_back_right02, bone2Name=dyn_seaweed_back_right01, swing1LimitDegrees=20, swing2LimitDegrees=10, twistLimitDegrees=1}
 {bone1Name=dyn_seaweed_back_right03, bone2Name=dyn_seaweed_back_right02, swing1LimitDegrees=20, swing2LimitDegrees=15, twistLimitDegrees=1}
 {bone1Name=dyn_hose_front, bone2Name=spine_03, swing1LimitDegrees=10, swing2LimitDegrees=1, twistLimitDegrees=1}
 {
  bone1Name=spine_03
  bone2Name=pelvis
  swing1LimitDegrees=45
  swing1Motion=Free
  swing2LimitDegrees=45
  swing2Motion=Free
  twistLimitDegrees=45
  twistMotion=Free
 }
]
