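# Animation Blueprint for the F_MED_Nautilus player body, adding secondary motion and fixes to its hoses, weight belt and seaweed.
# It copies the attached parent's pose, then runs two root-space RigidBody simulations using separate physics assets.
# Weighted constraints, Look At nodes and additive bone offsets then adjust the belt, front hose and seaweed.
# Each update, skydive or parachute state halves the seaweed Look At alpha, while knee-slide and emote states enable seaweed offsets.
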AnimBlueprint /BRCosmetics/Characters/Player/Female/Medium/Bodies/F_MED_Nautilus/Meshes/F_MED_Nautilus_AnimBP
target_skeleton=&/F_MED_Nautilus_Skeleton
graphs [
 AnimGraph {nodes [
   n1 = RigidBody(ComponentPose=Local To Component(LocalPose=Copy Pose From Mesh() {bUseAttachedParent=true}), Alpha=0.3) {
    bOverrideWorldGravity=true
    overridePhysicsAsset=&/F_MED_Nautilus_Physics
    simulationSpace=BaseBoneSpace
    alphaBoolBlend {blendOption=Linear}
    baseBoneRef {boneName=root}
    overrideWorldGravity {z=-980}
   }
   n2 = RigidBody(ComponentPose=@n1) {
    bOverrideWorldGravity=true
    overridePhysicsAsset=&/F_MED_Nautilus_hoses_weeds_Physics
    simulationSpace=BaseBoneSpace
    alphaBoolBlend {blendOption=Linear}
    baseBoneRef {boneName=root}
    overrideWorldGravity {z=-980}
   }
   n3 = "Constraint - dyn_weight_belt_back_ring_const (PPPPPPPP)"!(
    ComponentPose=@n2
    Alpha=1
    ConstraintWeights_0=0.0219
    ConstraintWeights_1:float=0.0667
    ConstraintWeights_2:float=0.1718
    ConstraintWeights_3:float=0.2379
    ConstraintWeights_4:float=0.4123
    ConstraintWeights_5:float=0.0696
    ConstraintWeights_6:float=0.0182
    ConstraintWeights_7:float=0.0015
   ) -> (Pose) {
    alphaBoolBlend {blendOption=Linear}
    boneToModify {boneName=dyn_weight_belt_back_ring_const}
    constraintSetup [
     {targetBone {boneName=clavicle_l}}
     {targetBone {boneName=spine_01}}
     {targetBone {boneName=spine_02}}
     {targetBone {boneName=spine_03}}
     {targetBone {boneName=spine_04}}
     {targetBone {boneName=spine_05}}
     {targetBone {boneName=clavicle_r}}
     {targetBone {boneName=pelvis}}
    ]
    constraintWeights [1,1,1,1,1,1,1,1]
   }
   n4 = "Constraint - dyn_hose_front_const (PPPPP)"!(
    ComponentPose=@n3
    Alpha=1
    ConstraintWeights_0=0.0416
    ConstraintWeights_1:float=0.0784
    ConstraintWeights_2:float=0.0183
    ConstraintWeights_3:float=0.8513
    ConstraintWeights_4:float=0.0105
   ) -> (Pose) {
    alphaBoolBlend {blendOption=Linear}
    boneToModify {boneName=dyn_hose_front_const}
    constraintSetup [
     {targetBone {boneName=spine_01}}
     {targetBone {boneName=spine_02}}
     {targetBone {boneName=spine_03}}
     {targetBone {boneName=spine_04}}
     {targetBone {boneName=pelvis}}
    ]
    constraintWeights [1,1,1,1,1]
   }
   n5 = Look At - Bone: dyn_belt_ring_front_top(ComponentPose=@n4, Alpha=1) {
    alphaBoolBlend {blendOption=Linear}
    boneToModify {boneName=dyn_belt_ring_front_top}
    lookAtLocation=<0 2>
    lookAtTarget {boneReference {boneName=dyn_weight_belt_front}}
    lookAt_Axis {axis=<1 0>}
   }
   n6 = Look At - Bone: dyn_front_seaweed_strand01(ComponentPose=@n5, Alpha=Get SkydiveAlpha() -> (SkydiveAlpha:double)) {
    alphaBoolBlend {blendOption=Linear}
    alphaScaleBiasClamp {bInterpResult=true}
    boneToModify {boneName=dyn_front_seaweed_strand01}
    lookAtLocation=<13.3223 9.0344 4.5083>
    lookAtTarget {boneReference {boneName=calf_l}}
    lookAt_Axis {axis=<1 0>}
   }
   n7 = "Transform (Modify) Bone - Bone: dyn_front_seaweed_strand02"!(ComponentPose=@n6, Alpha=Get KneeSlides() -> (KneeSlides:double), Rotation="0,-35,0") -> (Pose) {
    rotationMode=BMM_Additive
    rotationSpace=BCS_BoneSpace
    translationMode=BMM_Additive
    alphaBoolBlend {blendOption=Linear}
    alphaScaleBiasClamp {bInterpResult=true}
    boneToModify {boneName=dyn_front_seaweed_strand02}
    rotation {yaw=-35}
   }
   n8 = Get Emoting() -> (Emoting:double)
   n9 = "Transform (Modify) Bone - Bone: dyn_front_seaweed_strand02"!(ComponentPose=@n7, Alpha=@n8, Translation:Vector="0,0,-2", Rotation="0,-15,0") -> (Pose) {
    rotationMode=BMM_Additive
    rotationSpace=BCS_BoneSpace
    translationMode=BMM_Additive
    alphaBoolBlend {blendOption=Linear}
    alphaScaleBiasClamp {bInterpResult=true}
    boneToModify {boneName=dyn_front_seaweed_strand02}
    rotation {yaw=-15}
    translation {z=-2}
   }
   n10 = "Transform (Modify) Bone - Bone: dyn_front_seaweed_strand03"!(ComponentPose=@n9, Alpha=@n8, Translation="0,0,-2", Rotation="0,-10,-0.001") -> (Pose) {
    rotationMode=BMM_Additive
    rotationSpace=BCS_BoneSpace
    translationMode=BMM_Additive
    alphaBoolBlend {blendOption=Linear}
    alphaScaleBiasClamp {bInterpResult=true}
    boneToModify {boneName=dyn_front_seaweed_strand03}
    rotation {roll=-0.001, yaw=-10}
    translation {z=-2}
   }
   Output Pose(Result=Component To Local(ComponentPose=@n10)) {layerGroup=None}
  ]}
 EventGraph {nodes [
   Event BlueprintUpdateAnimation()
   > n1 = Sequence!(execute) -> (then_0, then_1, then_2:Exec)
   > n2 = Branch(Condition=OR Boolean!(A=Get bIsSkydiveDiveMode(), B=Get bIsSkydiveFloating(), C:bool=Get bIsSkyDiving(), D:bool=Get bIsParachuteOpened()) -> (ReturnValue))
   > Set SkydiveAlpha(SkydiveAlpha:double=0.5) -> (Output_Get:double)
   Set SkydiveAlpha(execute=@n2.else, SkydiveAlpha=1)
   Set Emoting(execute=@n1.then_1, Emoting:double="To Float (Boolean)"(InBool=Get bIsPlayingEmote())) -> (Output_Get:double)
   > Set KneeSlides(KneeSlides:double="To Float (Boolean)"(InBool=Get bIsKneeSliding())) -> (Output_Get:double)
   TryGetPawnOwner()
  ]}
]
variables [
 Emoting {type=number}
 KneeSlides {type=number}
 SkydiveAlpha {type=number}
]
